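# Unlit debug material for water that visualizes depth and flow from the WaterVelocityAndHeight texture, mapped over the landscape via MPC_Landscape.
# Inverted alpha becomes depth, shaped by a custom exponential falloff into a blue tint that also scales the velocity driving a flow-mapped grid.
# Depth values are overlaid as world-tiled debug digits.
# The blue channel yields a vertical offset toward the stored height, gated by MPC_Water's EnableTerrainMapping.
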Material /Clyde/Materials/Debug/WaterDepth
graph {
 properties {bUsedWithInstancedStaticMeshes=true, bUsedWithWater=true, pixelDepthOffsetMode=PDOM_Legacy, shadingModel=MSM_Unlit, usedShadingModels=Unlit}
 nodes [
  n1 = WorldPosition()
  n2 = Subtract(
   A=ComponentMask(None=@n1.XYZ) {g=true, r=true}
   B=ComponentMask(None=CollectionParameter() {collection=/Landmass/Landscape/BlueprintBrushes/MPC/MPC_Landscape, parameterName=LandscapeLocation}) {g=true, r=true}
  )
  n3 = ComponentMask(None=CollectionParameter() {collection=&MPC_Landscape, parameterName=RTWorldSize}) {g=true, r=true}
  n4 = TextureSampleParameter2D(UVs=Saturate(None=Divide(A=@n2, B=@n3))) {parameterName=WaterVelocityAndHeight, texture=&Black}
  n5 = Multiply(A=OneMinus(None=@n4.A)) {constB=1000}
  MaterialExpressionDivide_1 = Divide(A=@n5, B=ScalarParameter() {defaultValue=256, parameterName=Depth Gradient Strength})
  n6 = Custom!(x=@MaterialExpressionDivide_1) {
   code="exp(-x);"
   description=exp
   outputType=CMOT_Float1
   inputs [{inputName=x, input {expression=&self:MaterialExpressionDivide_1}}]
  }
  n7 = OneMinus(None=@n6)
  n8 = Multiply(A=Multiply(A=ComponentMask(None=@n4.RGB) {g=true, r=true}, B=Constant2Vector() {g=-1, r=1}), B=@n7)
  n9 = ScalarParameter() {defaultValue=1, parameterName=FLow}
  n10 = ScalarParameter() {defaultValue=512, parameterName=GridSize}
  n11 = MaterialFunctionCall!(
   Diffuse=TextureObject() {texture=/Engine/EditorMaterials/T_1x1_Grid}
   Normal Map
   "Flow Vector Map (see tooltip)"=Divide(A=Multiply(A=@n8, B=@n9), B=@n10)
   UV texture 2 offset
   "Time (see tooltip)"=MaterialFunctionCall(Speed=ScalarParameter() {defaultValue=1, parameterName=flowspeed}) {materialFunction=&TimeWithSpeedVariable}.Time
   UVs=Divide(A=WorldPosition().XYZ, B=@n10)
  ) {materialFunction=/Engine/Functions/Engine_MaterialFunctions02/Texturing/FlowMaps_Simple}
  n12 = ScalarParameter() {defaultValue=0.01, parameterName=Grid}
  n13 = VectorParameter() {parameterName=Blue, defaultValue=<rgba 0 0.1 0.3 0>}
  n14 = Frac(None=MaterialFunctionCall!(Scale=ScalarParameter() {defaultValue=512, parameterName=Text Scale}) {materialFunction=/Engine/Functions/Engine_MaterialFunctions01/Coordinates/WorldPosition-XY}.Result)
  Add(
   A=Add(A=Multiply(A=@n11.Diffuse, B=@n12), B=Multiply(A=@n13.RGB, B=@n7))
   B=ComponentMask(None=MaterialFunctionCall!(Vector4=@n5, MaximumNumberOfDigits, UVs=@n14, DebugTextLocation RG_UpperRight BA_LowerLeft, Component Spacing) {materialFunction=/Engine/Functions/Engine_MaterialFunctions02/Utility/DebugFloat4Values}.ColorCodedOutput) {b=true, g=true, r=true}
  )
  MaterialFunctionCall(Number=@n5, UVs=@n14) {materialFunction=&DebugScalarValues}
  n15 = Floor(None=@n4.B)
  n16 = ComponentMask(None=@n1.XYZ) {b=true}
  Multiply(
   A=Multiply(A=Subtract(A=@n15, B=@n16), B=Constant3Vector() {constant=<rgba 0 0 1 0>})
   B=CollectionParameter() {collection=/Water/Materials/MPC/MPC_Water, parameterName=EnableTerrainMapping}
  )
 ]
}
